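# Post-pose AnimBP for the HangedGrit male player body, adding secondary motion to jacket, collar, straps, zippers, necklace and knife.
# The parent mesh pose feeds a Control Rig, then chained AnimDynamics nodes, weighted constraints on sleeve zipper bases and an attach_backpack offset.
# EventGraph sets fixed collar and strap alphas.
# It also computes gravity, LOD, front-end, rigid-body and matching back-bling values, which the AnimGraph never reads.
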AnimBlueprint /BRCosmetics/Characters/Player/Male/Medium/Bodies/M_MED_HangedGrit/Meshes/M_MED_HangedGrit_AnimBP
target_skeleton=&/M_MED_HangedGrit_Skeleton
graphs [
 AnimGraph {nodes [
   Component To Local()
   Local To Component()
   n1 = Control Rig(Source=Copy Pose From Mesh() {bUseAttachedParent=true}) {
    alphaBoolBlend {blendOption=Linear}
    controlRigAssetReference {blueprintRigClass=&/M_MED_HangedGrit_CtrlRig.M_MED_HangedGrit_CtrlRig_C}
   }
   n2 = Anim Dynamics - Bone: dyn_jacket_fr_01_r!(ComponentPose=Local To Component(LocalPose=@n1), Alpha=1, GravityOverride:Vector="0,0,-980", ExternalForce="0,0,0") -> (Pose) {
    angularDampingOverride=0.9999
    angularSpringConstant=50
    bAngularSpring=true
    bOverrideAngularDamping=true
    bUseGravityOverride=true
    simulationSpace=RootRelative
    alphaBoolBlend {blendOption=Linear}
    boundBone {boneName=dyn_jacket_fr_01_r}
    componentAppliedLinearAccClamp=<5000 5000 5000>
    componentLinearAccScale=<0.5 0.5 0.5>
    componentLinearVelScale=<0.5 0.5 0.5>
    gravityOverride {z=-980}
    physicsBodyDefinitions [{
      boundBone {boneName=dyn_jacket_fr_01_r}
      boxExtents=<15 15 15>
      constraintSetup {angularLimitsMax=<2 3 6>, angularLimitsMin=<-2 -35 -6>, angularTarget=<1 0 0>}
      localJointOffset=<1 0 0>
     }]
   }
   n3 = Anim Dynamics - Bone: dyn_jacket_fr_02_r!(ComponentPose=@n2, Alpha=1, GravityOverride:Vector="0,0,-980", ExternalForce="0,0,0") -> (Pose) {
    angularDampingOverride=1
    angularSpringConstant=50
    bAngularSpring=true
    bOverrideAngularDamping=true
    bUseGravityOverride=true
    simulationSpace=RootRelative
    alphaBoolBlend {blendOption=Linear}
    boundBone {boneName=dyn_jacket_fr_02_r}
    componentAppliedLinearAccClamp=<5000 5000 5000>
    componentLinearAccScale=<0.5 0.5 0.5>
    componentLinearVelScale=<0.5 0.5 0.5>
    gravityOverride {z=-980}
    physicsBodyDefinitions [{
      boundBone {boneName=dyn_jacket_fr_02_r}
      boxExtents=<15 15 15>
      constraintSetup {angularLimitsMax=<2 3 6>, angularLimitsMin=<-2 -35 -6>, angularTarget=<1 0 0>}
      localJointOffset=<3 0 0>
     }]
   }
   n4 = Anim Dynamics - Bone: dyn_jacket_side_01_r!(ComponentPose=@n3, Alpha=1, GravityOverride:Vector="0,0,-980", ExternalForce="0,0,0") -> (Pose) {
    angularDampingOverride=0.9999
    angularSpringConstant=50
    bAngularSpring=true
    bOverrideAngularDamping=true
    bUseGravityOverride=true
    simulationSpace=RootRelative
    alphaBoolBlend {blendOption=Linear}
    boundBone {boneName=dyn_jacket_side_01_r}
    componentAppliedLinearAccClamp=<5000 5000 5000>
    componentLinearAccScale=<0.5 0.5 0.5>
    componentLinearVelScale=<0.5 0.5 0.5>
    gravityOverride {z=-980}
    physicsBodyDefinitions [{
      boundBone {boneName=dyn_jacket_side_01_r}
      boxExtents=<15 15 15>
      constraintSetup {angularLimitsMax=<2 1 5>, angularLimitsMin=<-2 -30 -5>, angularTarget=<1 0 0>}
      localJointOffset=<5 0 0>
     }]
   }
   n5 = Anim Dynamics - Bone: dyn_jacket_backside_01_r!(ComponentPose=@n4, Alpha=1, GravityOverride:Vector="0,0,-980", ExternalForce="0,0,0") -> (Pose) {
    angularDampingOverride=0.9999
    angularSpringConstant=50
    bAngularSpring=true
    bOverrideAngularDamping=true
    bUseGravityOverride=true
    simulationSpace=RootRelative
    alphaBoolBlend {blendOption=Linear}
    boundBone {boneName=dyn_jacket_backside_01_r}
    componentAppliedLinearAccClamp=<5000 5000 5000>
    componentLinearAccScale=<0.5 0.5 0.5>
    componentLinearVelScale=<0.5 0.5 0.5>
    gravityOverride {z=-980}
    physicsBodyDefinitions [{
      boundBone {boneName=dyn_jacket_backside_01_r}
      boxExtents=<15 15 15>
      constraintSetup {angularLimitsMax=<2 0 4>, angularLimitsMin=<-2 -25 -4>, angularTarget=<1 0 0>}
      localJointOffset=<0.5 0 0>
     }]
   }
   n6 = Anim Dynamics - Bone: dyn_jacket_back_01!(ComponentPose=@n5, Alpha=1, GravityOverride:Vector="0,0,-980", ExternalForce="0,0,0") -> (Pose) {
    angularDampingOverride=0.9999
    angularSpringConstant=50
    bAngularSpring=true
    bOverrideAngularDamping=true
    bUseGravityOverride=true
    simulationSpace=RootRelative
    alphaBoolBlend {blendOption=Linear}
    boundBone {boneName=dyn_jacket_back_01}
    componentAppliedLinearAccClamp=<5000 5000 5000>
    componentLinearAccScale=<0.5 0.5 0.5>
    componentLinearVelScale=<0.5 0.5 0.5>
    gravityOverride {z=-980}
    physicsBodyDefinitions [{
      boundBone {boneName=dyn_jacket_back_01}
      boxExtents=<15 15 15>
      constraintSetup {angularLimitsMax=<2 0 2>, angularLimitsMin=<-2 -15 -2>, angularTarget=<1 0 0>}
      localJointOffset=<0.5 0 0>
     }]
   }
   n7 = Anim Dynamics - Bone: dyn_jacket_fr_01_l!(ComponentPose=@n6, Alpha=1, GravityOverride:Vector="0,0,-980", ExternalForce="0,0,0") -> (Pose) {
    angularDampingOverride=0.9999
    angularSpringConstant=50
    bAngularSpring=true
    bOverrideAngularDamping=true
    bUseGravityOverride=true
    simulationSpace=RootRelative
    alphaBoolBlend {blendOption=Linear}
    boundBone {boneName=dyn_jacket_fr_01_l}
    componentAppliedLinearAccClamp=<5000 5000 5000>
    componentLinearAccScale=<0.5 0.5 0.5>
    componentLinearVelScale=<0.5 0.5 0.5>
    gravityOverride {z=-980}
    physicsBodyDefinitions [{
      boundBone {boneName=dyn_jacket_fr_01_l}
      boxExtents=<15 15 15>
      constraintSetup {angularLimitsMax=<2 3 6>, angularLimitsMin=<-2 -35 -6>, angularTarget=<1 0 0>}
      localJointOffset=<-1 0 0>
     }]
   }
   n8 = Anim Dynamics - Bone: dyn_jacket_fr_02_l!(ComponentPose=@n7, Alpha=1, GravityOverride:Vector="0,0,-980", ExternalForce="0,0,0") -> (Pose) {
    angularDampingOverride=1
    angularSpringConstant=50
    bAngularSpring=true
    bOverrideAngularDamping=true
    bUseGravityOverride=true
    simulationSpace=RootRelative
    alphaBoolBlend {blendOption=Linear}
    boundBone {boneName=dyn_jacket_fr_02_l}
    componentAppliedLinearAccClamp=<5000 5000 5000>
    componentLinearAccScale=<0.5 0.5 0.5>
    componentLinearVelScale=<0.5 0.5 0.5>
    gravityOverride {z=-980}
    physicsBodyDefinitions [{
      boundBone {boneName=dyn_jacket_fr_02_l}
      boxExtents=<15 15 15>
      constraintSetup {angularLimitsMax=<2 3 6>, angularLimitsMin=<-2 -35 -6>, angularTarget=<1 0 0>}
      localJointOffset=<-3 0 0>
     }]
   }
   n9 = Anim Dynamics - Bone: dyn_jacket_side_01_l!(ComponentPose=@n8, Alpha=1, GravityOverride:Vector="0,0,-980", ExternalForce="0,0,0") -> (Pose) {
    angularDampingOverride=0.9999
    angularSpringConstant=50
    bAngularSpring=true
    bOverrideAngularDamping=true
    bUseGravityOverride=true
    simulationSpace=RootRelative
    alphaBoolBlend {blendOption=Linear}
    boundBone {boneName=dyn_jacket_side_01_l}
    componentAppliedLinearAccClamp=<5000 5000 5000>
    componentLinearAccScale=<0.5 0.5 0.5>
    componentLinearVelScale=<0.5 0.5 0.5>
    gravityOverride {z=-980}
    physicsBodyDefinitions [{
      boundBone {boneName=dyn_jacket_side_01_l}
      boxExtents=<15 15 15>
      constraintSetup {angularLimitsMax=<2 1 5>, angularLimitsMin=<-2 -30 -5>, angularTarget=<1 0 0>}
      localJointOffset=<-5 0 0>
     }]
   }
   n10 = Anim Dynamics - Bone: dyn_jacket_backside_01_l!(ComponentPose=@n9, Alpha=1, GravityOverride:Vector="0,0,-980", ExternalForce="0,0,0") -> (Pose) {
    angularDampingOverride=0.9999
    angularSpringConstant=50
    bAngularSpring=true
    bOverrideAngularDamping=true
    bUseGravityOverride=true
    simulationSpace=RootRelative
    alphaBoolBlend {blendOption=Linear}
    boundBone {boneName=dyn_jacket_backside_01_l}
    componentAppliedLinearAccClamp=<5000 5000 5000>
    componentLinearAccScale=<0.5 0.5 0.5>
    componentLinearVelScale=<0.5 0.5 0.5>
    gravityOverride {z=-980}
    physicsBodyDefinitions [{
      boundBone {boneName=dyn_jacket_backside_01_l}
      boxExtents=<15 15 15>
      constraintSetup {angularLimitsMax=<2 0 4>, angularLimitsMin=<-2 -25 -4>, angularTarget=<1 0 0>}
      localJointOffset=<-0.5 0 0>
     }]
   }
   n11 = Get Collar Dynamic() -> (Collar Dynamic:double)
   n12 = Anim Dynamics - Bone: dyn_collar_fr_01_r!(ComponentPose=@n10, Alpha=@n11, GravityOverride:Vector="0,0,0", ExternalForce="0,0,0") -> (Pose) {
    angularDampingOverride=1
    angularSpringConstant=1000
    bAngularSpring=true
    bOverrideAngularDamping=true
    bUseGravityOverride=true
    simulationSpace=RootRelative
    alphaBoolBlend {blendOption=Linear}
    boundBone {boneName=dyn_collar_fr_01_r}
    componentAppliedLinearAccClamp=<5000 5000 5000>
    componentLinearAccScale=<0.05 0.05 0.05>
    componentLinearVelScale=<0.05 0.05 0.05>
    physicsBodyDefinitions [{
      boundBone {boneName=dyn_collar_fr_01_r}
      boxExtents=<10 10 10>
      constraintSetup {angularLimitsMax=<3 8 3>, angularLimitsMin=<-3 -8 -3>, angularTarget=<1 0 0>}
      localJointOffset=<-0.1 0 0>
     }]
    simSpaceSettings {maxAngularAcceleration=100, maxAngularVelocity=100, simSpaceAngularAlpha=0.2}
   }
   n13 = Anim Dynamics - Bone: dyn_collar_fr_02_r!(ComponentPose=@n12, Alpha=@n11, GravityOverride:Vector="0,0,0", ExternalForce="0,0,0") -> (Pose) {
    angularDampingOverride=1
    angularSpringConstant=1000
    bAngularSpring=true
    bOverrideAngularDamping=true
    bUseGravityOverride=true
    simulationSpace=RootRelative
    alphaBoolBlend {blendOption=Linear}
    boundBone {boneName=dyn_collar_fr_02_r}
    componentAppliedLinearAccClamp=<5000 5000 5000>
    componentLinearAccScale=<0.05 0.05 0.05>
    componentLinearVelScale=<0.05 0.05 0.05>
    physicsBodyDefinitions [{
      boundBone {boneName=dyn_collar_fr_02_r}
      boxExtents=<10 10 10>
      constraintSetup {angularLimitsMax=<3 11 3>, angularLimitsMin=<-3 -11 -3>, angularTarget=<1 0 0>}
      localJointOffset=<-0.1 0 0>
     }]
    simSpaceSettings {maxAngularAcceleration=100, maxAngularVelocity=100, simSpaceAngularAlpha=0.2}
   }
   n14 = Anim Dynamics - Bone: dyn_collar_dw_fr_01_r!(ComponentPose=@n13, Alpha=@n11, GravityOverride:Vector="0,0,0", ExternalForce="0,0,0") -> (Pose) {
    angularDampingOverride=1
    angularSpringConstant=1000
    bAngularSpring=true
    bOverrideAngularDamping=true
    bUseGravityOverride=true
    simulationSpace=RootRelative
    alphaBoolBlend {blendOption=Linear}
    boundBone {boneName=dyn_collar_dw_fr_01_r}
    componentAppliedLinearAccClamp=<5000 5000 5000>
    componentLinearAccScale=<0.05 0.05 0.05>
    componentLinearVelScale=<0.05 0.05 0.05>
    physicsBodyDefinitions [{
      boundBone {boneName=dyn_collar_dw_fr_01_r}
      boxExtents=<10 10 10>
      constraintSetup {angularLimitsMax=<3 15 3>, angularLimitsMin=<-3 -15 -3>, angularTarget=<1 0 0>}
      localJointOffset=<-0.1 0 0>
     }]
    simSpaceSettings {maxAngularAcceleration=100, maxAngularVelocity=100, simSpaceAngularAlpha=0.2}
   }
   n15 = Anim Dynamics - Bone: dyn_collar_side_01_r!(ComponentPose=@n14, Alpha=@n11, GravityOverride:Vector="0,0,0", ExternalForce="0,0,0") -> (Pose) {
    angularDampingOverride=1
    angularSpringConstant=1000
    bAngularSpring=true
    bOverrideAngularDamping=true
    bUseGravityOverride=true
    simulationSpace=RootRelative
    alphaBoolBlend {blendOption=Linear}
    boundBone {boneName=dyn_collar_side_01_r}
    componentAppliedLinearAccClamp=<5000 5000 5000>
    componentLinearAccScale=<0.05 0.05 0.05>
    componentLinearVelScale=<0.05 0.05 0.05>
    physicsBodyDefinitions [{
      boundBone {boneName=dyn_collar_side_01_r}
      boxExtents=<10 10 10>
      constraintSetup {angularLimitsMax=<3 10 3>, angularLimitsMin=<-3 -10 -3>, angularTarget=<1 0 0>}
      localJointOffset=<-0.1 0 0>
     }]
    simSpaceSettings {maxAngularAcceleration=100, maxAngularVelocity=100, simSpaceAngularAlpha=0.2}
   }
   n16 = Anim Dynamics - Bone: dyn_collar_back_01!(ComponentPose=@n15, Alpha=@n11, GravityOverride:Vector="0,0,0", ExternalForce="0,0,0") -> (Pose) {
    angularDampingOverride=1
    angularSpringConstant=1000
    bOverrideAngularDamping=true
    bUseGravityOverride=true
    simulationSpace=RootRelative
    alphaBoolBlend {blendOption=Linear}
    boundBone {boneName=dyn_collar_back_01}
    componentAppliedLinearAccClamp=<5000 5000 5000>
    componentLinearAccScale=<0.05 0.05 0.05>
    componentLinearVelScale=<0.05 0.05 0.05>
    physicsBodyDefinitions [{
      boundBone {boneName=dyn_collar_back_01}
      boxExtents=<10 10 10>
      constraintSetup {angularLimitsMax=<2 8 2>, angularLimitsMin=<-2 -3 -2>, angularTarget=<1 0 0>}
      localJointOffset=<-0.1 0 0>
     }]
    simSpaceSettings {maxAngularAcceleration=100, maxAngularVelocity=100, simSpaceAngularAlpha=0.2}
   }
   n17 = Anim Dynamics - Bone: dyn_collar_fr_01_l!(ComponentPose=@n16, Alpha=@n11, GravityOverride:Vector="0,0,0", ExternalForce="0,0,0") -> (Pose) {
    angularDampingOverride=1
    angularSpringConstant=1000
    bAngularSpring=true
    bOverrideAngularDamping=true
    bUseGravityOverride=true
    simulationSpace=RootRelative
    alphaBoolBlend {blendOption=Linear}
    boundBone {boneName=dyn_collar_fr_01_l}
    componentAppliedLinearAccClamp=<5000 5000 5000>
    componentLinearAccScale=<0.05 0.05 0.05>
    componentLinearVelScale=<0.05 0.05 0.05>
    physicsBodyDefinitions [{
      boundBone {boneName=dyn_collar_fr_01_l}
      boxExtents=<10 10 10>
      constraintSetup {angularLimitsMax=<3 8 3>, angularLimitsMin=<-3 -8 -3>, angularTarget=<1 0 0>}
      localJointOffset=<0.1 0 0>
     }]
    simSpaceSettings {maxAngularAcceleration=100, maxAngularVelocity=100, simSpaceAngularAlpha=0.2}
   }
   n18 = Anim Dynamics - Bone: dyn_collar_fr_02_l!(ComponentPose=@n17, Alpha=@n11, GravityOverride:Vector="0,0,0", ExternalForce="0,0,0") -> (Pose) {
    angularDampingOverride=1
    angularSpringConstant=1000
    bAngularSpring=true
    bOverrideAngularDamping=true
    bUseGravityOverride=true
    simulationSpace=RootRelative
    alphaBoolBlend {blendOption=Linear}
    boundBone {boneName=dyn_collar_fr_02_l}
    componentAppliedLinearAccClamp=<5000 5000 5000>
    componentLinearAccScale=<0.05 0.05 0.05>
    componentLinearVelScale=<0.05 0.05 0.05>
    physicsBodyDefinitions [{
      boundBone {boneName=dyn_collar_fr_02_l}
      boxExtents=<10 10 10>
      constraintSetup {angularLimitsMax=<3 11 3>, angularLimitsMin=<-3 -11 -3>, angularTarget=<1 0 0>}
      localJointOffset=<0.1 0 0>
     }]
    simSpaceSettings {maxAngularAcceleration=100, maxAngularVelocity=100, simSpaceAngularAlpha=0.2}
   }
   n19 = Anim Dynamics - Bone: dyn_collar_dw_fr_01_l!(ComponentPose=@n18, Alpha=@n11, GravityOverride:Vector="0,0,0", ExternalForce="0,0,0") -> (Pose) {
    angularDampingOverride=1
    angularSpringConstant=1000
    bAngularSpring=true
    bOverrideAngularDamping=true
    bUseGravityOverride=true
    simulationSpace=RootRelative
    alphaBoolBlend {blendOption=Linear}
    boundBone {boneName=dyn_collar_dw_fr_01_l}
    componentAppliedLinearAccClamp=<5000 5000 5000>
    componentLinearAccScale=<0.05 0.05 0.05>
    componentLinearVelScale=<0.05 0.05 0.05>
    physicsBodyDefinitions [{
      boundBone {boneName=dyn_collar_dw_fr_01_l}
      boxExtents=<10 10 10>
      constraintSetup {angularLimitsMax=<3 15 3>, angularLimitsMin=<-3 -15 -3>, angularTarget=<1 0 0>}
      localJointOffset=<0.1 0 0>
     }]
    simSpaceSettings {maxAngularAcceleration=100, maxAngularVelocity=100, simSpaceAngularAlpha=0.2}
   }
   n20 = Anim Dynamics - Bone: dyn_collar_side_01_l!(ComponentPose=@n19, Alpha=@n11, GravityOverride:Vector="0,0,0", ExternalForce="0,0,0") -> (Pose) {
    angularDampingOverride=1
    angularSpringConstant=1000
    bAngularSpring=true
    bOverrideAngularDamping=true
    bUseGravityOverride=true
    simulationSpace=RootRelative
    alphaBoolBlend {blendOption=Linear}
    boundBone {boneName=dyn_collar_side_01_l}
    componentAppliedLinearAccClamp=<5000 5000 5000>
    componentLinearAccScale=<0.05 0.05 0.05>
    componentLinearVelScale=<0.05 0.05 0.05>
    physicsBodyDefinitions [{
      boundBone {boneName=dyn_collar_side_01_l}
      boxExtents=<10 10 10>
      constraintSetup {angularLimitsMax=<3 10 3>, angularLimitsMin=<-3 -10 -3>, angularTarget=<1 0 0>}
      localJointOffset=<0.1 0 0>
     }]
    simSpaceSettings {maxAngularAcceleration=100, maxAngularVelocity=100, simSpaceAngularAlpha=0.2}
   }
   n21 = Get Strap Dynamic() -> (Strap Dynamic:double)
   n22 = Anim Dynamics - Bone: dyn_strap_collar_01_r!(ComponentPose=@n20, Alpha=@n21, GravityOverride:Vector="0,0,0", ExternalForce="0,0,0") -> (Pose) {
    angularDampingOverride=0.9999
    angularSpringConstant=0.05
    bAngularSpring=true
    bOverrideAngularDamping=true
    bUseGravityOverride=true
    simulationSpace=World
    alphaBoolBlend {blendOption=Linear}
    boundBone {boneName=dyn_strap_collar_01_r}
    componentAppliedLinearAccClamp=<5000 5000 5000>
    componentLinearAccScale=<0.05 0.05 0.05>
    componentLinearVelScale=<0.05 0.05 0.05>
    physicsBodyDefinitions [{
      boundBone {boneName=dyn_strap_collar_01_r}
      boxExtents=<10 10 10>
      constraintSetup {angularTargetAxis=AxisZ, twistAxis=AxisZ, angularLimitsMax=<5 3 10>, angularLimitsMin=<-20 -3 -10>, angularTarget=<0 -1 0>}
      localJointOffset=<0 -3 0>
     }]
    simSpaceSettings {maxAngularAcceleration=100, maxAngularVelocity=100, simSpaceAngularAlpha=0.2}
   }
   n23 = Anim Dynamics - Bone: dyn_strap_collar_02_r!(ComponentPose=@n22, Alpha=@n21, GravityOverride:Vector="0,0,0", ExternalForce="0,0,0") -> (Pose) {
    angularDampingOverride=0.9999
    angularSpringConstant=75
    bAngularSpring=true
    bOverrideAngularDamping=true
    bUseGravityOverride=true
    simulationSpace=World
    alphaBoolBlend {blendOption=Linear}
    boundBone {boneName=dyn_strap_collar_02_r}
    componentAppliedLinearAccClamp=<5000 5000 5000>
    componentLinearAccScale=<0.05 0.05 0.05>
    componentLinearVelScale=<0.05 0.05 0.05>
    physicsBodyDefinitions [{
      boundBone {boneName=dyn_strap_collar_02_r}
      boxExtents=<10 10 10>
      constraintSetup {angularTargetAxis=AxisZ, twistAxis=AxisZ, angularLimitsMax=<5 3 3>, angularLimitsMin=<-20 -3 -3>, angularTarget=<0 -1 0>}
      localJointOffset=<0 -4 0>
     }]
    simSpaceSettings {maxAngularAcceleration=100, maxAngularVelocity=100, simSpaceAngularAlpha=0.2}
   }
   n24 = Anim Dynamics - Bone: dyn_strap_collar_01_l!(ComponentPose=@n23, Alpha=@n21, GravityOverride:Vector="0,0,0", ExternalForce="0,0,0") -> (Pose) {
    angularDampingOverride=0.9999
    angularSpringConstant=200
    bAngularSpring=true
    bOverrideAngularDamping=true
    bUseGravityOverride=true
    simulationSpace=World
    alphaBoolBlend {blendOption=Linear}
    boundBone {boneName=dyn_strap_collar_01_l}
    componentAppliedLinearAccClamp=<5000 5000 5000>
    componentLinearAccScale=<0.05 0.05 0.05>
    componentLinearVelScale=<0.05 0.05 0.05>
    physicsBodyDefinitions [{
      boundBone {boneName=dyn_strap_collar_01_l}
      boxExtents=<10 10 10>
      constraintSetup {angularTargetAxis=AxisZ, twistAxis=AxisZ, angularLimitsMax=<3 3 3>, angularLimitsMin=<-45 -3 -3>, angularTarget=<0 -1 0>}
      localJointOffset=<0 -3 0>
     }]
    simSpaceSettings {maxAngularAcceleration=100, maxAngularVelocity=100, simSpaceAngularAlpha=0.2}
   }
   n25 = "Constraint - dyn_zip_sleeve_base_l (PPPPPPP)"!(
    ComponentPose=@n24
    Alpha=1
    ConstraintWeights_0=0
    ConstraintWeights_1:float=0.2935
    ConstraintWeights_2:float=0.003
    ConstraintWeights_3:float=0.0428
    ConstraintWeights_4:float=0.0606
    ConstraintWeights_5:float=0.54
    ConstraintWeights_6:float=0.0601
   ) -> (Pose) {
    alphaBoolBlend {blendOption=Linear}
    boneToModify {boneName=dyn_zip_sleeve_base_l}
    constraintSetup [
     {targetBone {boneName=upperarm_l}}
     {targetBone {boneName=lowerarm_l}}
     {targetBone {boneName=hand_l}}
     {targetBone {boneName=deform_wrist_up_l}}
     {targetBone {boneName=deform_wrist_dn_l}}
     {targetBone {boneName=lowerarm_twist_01_l}}
     {targetBone {boneName=lowerarm_twist_02_l}}
    ]
    constraintWeights [1,1,1,1,1,1,1]
   }
   n26 = "Constraint - dyn_zip_sleeve_base_r (PPPPPPP)"!(
    ComponentPose=@n25
    Alpha=1
    ConstraintWeights_0=0
    ConstraintWeights_1:float=0.2953
    ConstraintWeights_2:float=0.003
    ConstraintWeights_3:float=0.0428
    ConstraintWeights_4:float=0.0621
    ConstraintWeights_5:float=0.5402
    ConstraintWeights_6:float=0.0567
   ) -> (Pose) {
    alphaBoolBlend {blendOption=Linear}
    boneToModify {boneName=dyn_zip_sleeve_base_r}
    constraintSetup [
     {targetBone {boneName=upperarm_r}}
     {targetBone {boneName=lowerarm_r}}
     {targetBone {boneName=hand_r}}
     {targetBone {boneName=deform_wrist_up_r}}
     {targetBone {boneName=deform_wrist_dn_r}}
     {targetBone {boneName=lowerarm_twist_01_r}}
     {targetBone {boneName=lowerarm_twist_02_r}}
    ]
    constraintWeights [1,1,1,1,1,1,1]
   }
   n27 = Anim Dynamics - Bone: dyn_zip_jacket_01!(ComponentPose=@n26, Alpha=1, GravityOverride:Vector="0,0,-980", ExternalForce="0,0,0") -> (Pose) {
    angularDampingOverride=0.9
    angularSpringConstant=20
    bAngularSpring=true
    bOverrideAngularDamping=true
    bUseGravityOverride=true
    simulationSpace=RootRelative
    alphaBoolBlend {blendOption=Linear}
    boundBone {boneName=dyn_zip_jacket_01}
    componentAppliedLinearAccClamp=<5000 5000 5000>
    componentLinearAccScale=<0.5 0.5 0.5>
    componentLinearVelScale=<0.5 0.5 0.5>
    gravityOverride {z=-980}
    physicsBodyDefinitions [{
      boundBone {boneName=dyn_zip_jacket_01}
      boxExtents=<15 15 15>
      constraintSetup {angularLimitsMax=<1 15 15>, angularLimitsMin=<-1 -140 -15>, angularTarget=<1 0 0>}
      localJointOffset=<1 0 0>
     }]
   }
   n28 = Anim Dynamics - Bone: dyn_zip_sleeve_01_r!(ComponentPose=@n27, Alpha=1, GravityOverride:Vector="0,0,-980", ExternalForce="0,0,0") -> (Pose) {
    angularDampingOverride=0.9
    angularSpringConstant=20
    bAngularSpring=true
    bOverrideAngularDamping=true
    bUseGravityOverride=true
    simulationSpace=RootRelative
    alphaBoolBlend {blendOption=Linear}
    boundBone {boneName=dyn_zip_sleeve_01_r}
    componentAppliedLinearAccClamp=<5000 5000 5000>
    componentLinearAccScale=<0.5 0.5 0.5>
    componentLinearVelScale=<0.5 0.5 0.5>
    gravityOverride {z=-980}
    physicsBodyDefinitions [{
      boundBone {boneName=dyn_zip_sleeve_01_r}
      boxExtents=<15 15 15>
      constraintSetup {angularLimitsMax=<1 15 15>, angularLimitsMin=<-1 -140 -15>, angularTarget=<1 0 0>}
      localJointOffset=<1 0 0>
     }]
   }
   n29 = Anim Dynamics - Bone: dyn_zip_sleeve_01_l!(ComponentPose=@n28, Alpha=1, GravityOverride:Vector="0,0,-980", ExternalForce="0,0,0") -> (Pose) {
    angularDampingOverride=0.9
    angularSpringConstant=20
    bAngularSpring=true
    bOverrideAngularDamping=true
    bUseGravityOverride=true
    simulationSpace=RootRelative
    alphaBoolBlend {blendOption=Linear}
    boundBone {boneName=dyn_zip_sleeve_01_l}
    componentAppliedLinearAccClamp=<5000 5000 5000>
    componentLinearAccScale=<0.5 0.5 0.5>
    componentLinearVelScale=<0.5 0.5 0.5>
    gravityOverride {z=-980}
    physicsBodyDefinitions [{
      boundBone {boneName=dyn_zip_sleeve_01_l}
      boxExtents=<15 15 15>
      constraintSetup {angularLimitsMax=<1 15 15>, angularLimitsMin=<-1 -140 -15>, angularTarget=<1 0 0>}
      localJointOffset=<-1 0 0>
     }]
   }
   n30 = Anim Dynamics - Bone: dyn_necklace_chain!(ComponentPose=@n29, Alpha=1, GravityOverride:Vector="0,0,-980", ExternalForce="0,0,0") -> (Pose) {
    angularDampingOverride=1
    angularSpringConstant=750
    bAngularSpring=true
    bOverrideAngularDamping=true
    bUseGravityOverride=true
    linearDampingOverride=1
    simulationSpace=RootRelative
    alphaBoolBlend {blendOption=Linear}
    boundBone {boneName=dyn_necklace_chain}
    componentAppliedLinearAccClamp=<5000 5000 5000>
    componentLinearAccScale=<0.05 0.05 0.05>
    componentLinearVelScale=<0.05 0.05 0.05>
    gravityOverride {z=-980}
    physicsBodyDefinitions [{
      boundBone {boneName=dyn_necklace_chain}
      boxExtents=<15 15 15>
      constraintSetup {angularLimitsMax=<1 0 8>, angularLimitsMin=<1 -20 -8>, angularTarget=<1 0 0>}
      localJointOffset=<0.5 0 0>
     }]
    simSpaceSettings {maxAngularAcceleration=100, maxAngularVelocity=100, simSpaceAngularAlpha=0.2}
   }
   n31 = Anim Dynamics - Bone: dyn_necklace_badge!(ComponentPose=@n30, Alpha=1, GravityOverride:Vector="0,0,-980", ExternalForce="0,0,0") -> (Pose) {
    angularDampingOverride=0.999
    angularSpringConstant=50
    bAngularSpring=true
    bOverrideAngularDamping=true
    bUseGravityOverride=true
    simulationSpace=RootRelative
    alphaBoolBlend {blendOption=Linear}
    boundBone {boneName=dyn_necklace_badge}
    componentAppliedLinearAccClamp=<5000 5000 5000>
    componentLinearAccScale=<0.05 0.05 0.05>
    componentLinearVelScale=<0.05 0.05 0.05>
    gravityOverride {z=-980}
    physicsBodyDefinitions [{
      boundBone {boneName=dyn_necklace_badge}
      boxExtents=<15 15 15>
      constraintSetup {angularLimitsMax=<1 1 20>, angularLimitsMin=<-1 -60 -20>, angularTarget=<1 0 0>}
      localJointOffset=<2.5 0 0>
     }]
    simSpaceSettings {maxAngularAcceleration=100, maxAngularVelocity=100, simSpaceAngularAlpha=0.2}
   }
   n32 = Anim Dynamics - Bone: dyn_knife_01!(ComponentPose=@n31, Alpha=1, GravityOverride:Vector="0,0,0", ExternalForce="0,0,0") -> (Pose) {
    angularBiasOverride=1
    angularDampingOverride=1
    angularSpringConstant=1000
    bAngularSpring=true
    bLinearSpring=true
    bOverrideAngularBias=true
    bOverrideAngularDamping=true
    bUseGravityOverride=true
    linearSpringConstant=50
    simulationSpace=RootRelative
    alphaBoolBlend {blendOption=Linear}
    boundBone {boneName=dyn_knife_01}
    componentAppliedLinearAccClamp=<5000 5000 5000>
    componentLinearAccScale=<0.05 0.05 0.05>
    componentLinearVelScale=<0.05 0.05 0.05>
    physicsBodyDefinitions [{
      boundBone {boneName=dyn_knife_01}
      boxExtents=<10 10 10>
      constraintSetup {angularLimitsMax=<2 2 6>, angularLimitsMin=<-2 -2 -6>, angularTarget=<1 0 0>, linearAxesMin=<-0.5 0 0>}
      localJointOffset=<0.3 0 0>
     }]
    simSpaceSettings {maxAngularAcceleration=100, maxAngularVelocity=100, simSpaceAngularAlpha=0.2}
   }
   n33 = "Transform (Modify) Bone - Bone: attach_backpack"(ComponentPose=@n32, Alpha=1, Translation="-2,2,0", Rotation="10,0,8", Scale="1,1,1") {
    rotationMode=BMM_Additive
    rotationSpace=BCS_BoneSpace
    translationMode=BMM_Additive
    translationSpace=BCS_BoneSpace
    alphaBoolBlend {blendOption=Linear}
    boneToModify {boneName=attach_backpack}
    rotation {pitch=10, roll=8}
    translation=<-2 2>
   }
   Output Pose(Result=Component To Local(ComponentPose=@n33)) {layerGroup=None}
  ]}
 EventGraph {nodes [
   Event BlueprintUpdateAnimation()
   > n1 = Sequence!(execute) -> (then_0, then_1, then_2:Exec, then_3:Exec, then_4:Exec, then_5:Exec)
   > n2 = Gravity Override State Function DEPRECATED(
    CharacterPartAnimInstance=Self-Reference()
    GameTimeInSeconds=GetGameTimeInSeconds()
    GravityOverrideParams=Get Gravity Override Params()
    DebugJointName=Pelvis
   )
   > Set Gravity Override(Gravity Override=@n2)
   n3 = Get bIsPlayingEmote()
   > Set FrontBOOL(FrontBOOL:bool=AND Boolean(A=Get bIsInFrontEnd(), B=NOT Boolean(A=@n3))) -> (Output_Get:bool)
   n4 = Get CurrentLOD()
   Set LOD1(execute=@n1.then_1, LOD1:bool="integer >= integer"(A=@n4, B=1)) -> (Output_Get:bool)
   Set LOD2(execute=@n1.then_2, LOD2:bool="integer >= integer"(A=@n4, B=2)) -> (Output_Get:bool)
   Set RBAN_off(execute=@n1.then_3, RBAN_off:bool=Get bRigidBodyEnabled()) -> (Output_Get:bool)
   Set Strap Dynamic(execute=@n1.then_4, Strap Dynamic:double=0.5) -> (Output_Get:double)
   Set Collar Dynamic(execute=@n1.then_5, Collar Dynamic:double=0.7) -> (Output_Get:double)
   n5 = Event OnCharacterCustomizationLoaded()
   > n6 = For Each Loop(Array:Array of Custom Character Part Object References=@n5.CustomParts) -> (Array Element:Custom Character Part Object Reference)
   > Branch(Condition="Equal (Enum)"(
     A:EFortCustomPartType Enum=GetPartType!(self:Custom Character Part Object Reference=@n6) -> (ReturnValue:EFortCustomPartType Enum=Head)
     B:EFortCustomPartType Enum=Backpack
    ))
   > n7 = Branch(Condition=Contains(SearchIn=GetObjectName(Object=@n6), Substring=Get Substring() -> (Substring:String)))
   > Set IsThisMyBackBling(IsThisMyBackBling:bool=true) -> (Output_Get:bool)
   Set IsThisMyBackBling(execute=@n7.else, IsThisMyBackBling=false)
  ]}
]
variables [
 Collar Dynamic {type=number}
 FrontBOOL {type=boolean}
 Gravity Override {type=Vector}
 Gravity Override Params {
  type=GravityOverrideParamsStruct
  default {dBNO {jointName=root, gravityOverride=<0 0 -980>}, skydive_Parachuting {gravityOverride=<0 -500 -980>, windAmplitude=<200 -200 150>}}
 }
 IsThisMyBackBling {type=boolean}
 LOD1 {type=boolean}
 LOD2 {type=boolean}
 RBAN_off {type=boolean}
 Strap Dynamic {type=number}
 Substring {type=string, default=HangedGrit}
]
