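# Physics asset for the HangedGrit medium male character's face accessory part, covering a short dynamic hair lock attached to the head.
# Two small capsules on the lock bones form a chain, with a larger capsule on the head bone.
# Constraints link the first lock bone to the head and the second to the first, with tighter swing and twist limit angles at the root than at the tip.
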PhysicsAsset /BRCosmetics/Characters/Player/Male/Medium/Bodies/M_MED_HangedGrit/Meshes/Parts/M_MED_HangedGrit_FaceAcc_Physics
bodies {
 dyn_lockhair_01 [{shapeType=Capsule, length=2.332, radius=0.9157, center=<1.221 0.0001 -0.1294>, rotation=<rot -75.6143 179.9938 -179.9913>}]
 dyn_lockhair_02 [{shapeType=Capsule, length=2.332, radius=0.9157, center=<1.3807 0.0001 -0.8922>, rotation=<rot -51.4066 179.9995 -179.9966>}]
 head [{shapeType=Capsule, length=1, radius=8, center=<5.1362 1.241 0>}]
}
constraints [
 {bone1Name=dyn_lockhair_01, bone2Name=head, swing1LimitDegrees=15, swing2LimitDegrees=15, twistLimitDegrees=5}
 {bone1Name=dyn_lockhair_02, bone2Name=dyn_lockhair_01, swing1LimitDegrees=25, swing2LimitDegrees=15, twistLimitDegrees=25}
]
