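# Post-copy AnimBP for the HangedGrit face accessory, adding secondary motion to its hair on top of the pose copied from the attached parent mesh.
# A head-space RigidBody node feeds two sprung AnimDynamics lock-hair bones, enabled when rigid body is disabled, then a constant deform_hair_back offset and scale.
# EventGraph updates that toggle, plus gravity override, front-end flag and head-yaw alpha.
# The head-bone transform using that alpha is unconnected to the output pose.
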AnimBlueprint /BRCosmetics/Characters/Player/Male/Medium/Bodies/M_MED_HangedGrit/Meshes/Parts/M_MED_HangedGrit_FaceAcc_AnimBP
target_skeleton=&/M_MED_HangedGrit_FaceAcc_Skeleton
graphs [
 AnimGraph {nodes [
   "Transform (Modify) Bone - Bone: head"(Alpha=Get Rotate Head Alpha() -> (Rotate Head Alpha:double), Translation="0,0,0", Rotation="0,0,0", Scale="1,1,1") {
    rotationMode=BMM_Additive
    scaleMode=BMM_Additive
    translationMode=BMM_Additive
    alphaBoolBlend {blendOption=Linear}
    boneToModify {boneName=head}
   }
   n1 = RigidBody!(
    ComponentPose=Local To Component(LocalPose=Copy Pose From Mesh() {bUseAttachedParent=true})
    Alpha
    OverrideWorldGravity:Vector="0,0,0"
    ExternalForce
   ) -> (Pose) {
    bClampLinearTranslationLimitToRefPose=true
    bOverrideWorldGravity=true
    overridePhysicsAsset=&/M_MED_HangedGrit_FaceAcc_Physics
    simulationSpace=BaseBoneSpace
    alphaBoolBlend {blendOption=Linear}
    baseBoneRef {boneName=head}
    simSpaceSettings {maxAngularAcceleration=100, maxAngularVelocity=3, maxLinearAcceleration=100, maxLinearVelocity=75, velocityScaleZ=0, worldAlpha=0.9}
   }
   n2 = Get RBAN_Disabled() -> (RBAN_Disabled:bool)
   n3 = Anim Dynamics - Bone: dyn_lockhair_01!(ComponentPose=@n1, bAlphaBoolEnabled:bool=@n2, GravityOverride:Vector="0,0,-980", ExternalForce="0,0,0") -> (Pose) {
    alphaInputType=Bool
    angularDampingOverride=0.9
    angularSpringConstant=150
    bAngularSpring=true
    bOverrideAngularDamping=true
    bUseGravityOverride=true
    simulationSpace=RootRelative
    alphaBoolBlend {blendOption=Linear}
    boundBone {boneName=dyn_lockhair_01}
    componentAppliedLinearAccClamp=<5000 5000 5000>
    componentLinearAccScale=<0.5 0.5 0.5>
    componentLinearVelScale=<0.5 0.5 0.5>
    gravityOverride {z=-980}
    physicsBodyDefinitions [{
      boundBone {boneName=dyn_lockhair_01}
      boxExtents=<15 15 15>
      constraintSetup {angularLimitsMax=<2 5 5>, angularLimitsMin=<-2 -5 -5>, angularTarget=<1 0 0>}
      localJointOffset=<1 0 0>
     }]
   }
   n4 = Anim Dynamics - Bone: dyn_lockhair_02!(ComponentPose=@n3, bAlphaBoolEnabled:bool=@n2, GravityOverride:Vector="0,0,-980", ExternalForce="0,0,0") -> (Pose) {
    alphaInputType=Bool
    angularDampingOverride=0.75
    angularSpringConstant=150
    bAngularSpring=true
    bOverrideAngularDamping=true
    bUseGravityOverride=true
    simulationSpace=RootRelative
    alphaBoolBlend {blendOption=Linear}
    boundBone {boneName=dyn_lockhair_02}
    componentAppliedLinearAccClamp=<5000 5000 5000>
    componentLinearAccScale=<0.5 0.5 0.5>
    componentLinearVelScale=<0.5 0.5 0.5>
    gravityOverride {z=-980}
    physicsBodyDefinitions [{
      boundBone {boneName=dyn_lockhair_02}
      boxExtents=<15 15 15>
      constraintSetup {angularLimitsMax=<5 5 10>, angularLimitsMin=<-5 -10 -10>, angularTarget=<1 0 0>}
      localJointOffset=<1 0 0>
     }]
   }
   n5 = "Transform (Modify) Bone - Bone: deform_hair_back"(ComponentPose=@n4, Alpha=1, Translation="0,-0.1,0", Rotation="0,0,0", Scale="1,1.1,1.1") {
    rotationMode=BMM_Additive
    scaleMode=BMM_Additive
    translationMode=BMM_Additive
    alphaBoolBlend {blendOption=Linear}
    boneToModify {boneName=deform_hair_back}
    scale {y=1.1, z=1.1}
    translation {y=-0.1}
   }
   Output Pose(Result=Component To Local(ComponentPose=@n5)) {layerGroup=None}
  ]}
 EventGraph {nodes [
   Event BlueprintUpdateAnimation()
   > n1 = Sequence!(execute) -> (then_0, then_1, then_2:Exec, then_3:Exec, then_4:Exec, then_5:Exec)
   > n2 = Gravity Override State Function DEPRECATED(
    CharacterPartAnimInstance=Self-Reference()
    GameTimeInSeconds=GetGameTimeInSeconds()
    GravityOverrideParams=Get Gravity Override Params()
    DebugJointName=Pelvis
   )
   > Set Gravity Override(Gravity Override=@n2)
   n3 = Get bIsPlayingEmote()
   > Set FrontBOOL(FrontBOOL:bool=AND Boolean(A=Get bIsInFrontEnd(), B=NOT Boolean(A=@n3))) -> (Output_Get:bool)
   Set RBAN_Disabled(execute=@n1.then_1, RBAN_Disabled:bool=NOT Boolean(A=Get bRigidBodyEnabled())) -> (Output_Get:bool)
   n4 = BreakTransform(InTransform=GetDeltaTransformFromRefPose(self=GetOwnerSkeletalMeshComponent(), BoneName=head, BaseName=neck_02))
   Set Rotate Head Alpha(execute=@n1.then_2, Rotate Head Alpha:double=MapRangeClamped(Value=BreakRotator(InRot=@n4.Rotation).Yaw, InRangeA=-20, InRangeB=-45)) -> (Output_Get:double)
  ]}
]
variables [
 FrontBOOL {type=boolean}
 Gravity Override {type=Vector}
 Gravity Override Params {type=GravityOverrideParamsStruct}
 RBAN_Disabled {type=boolean}
 Rotate Head Alpha {type=number}
]
